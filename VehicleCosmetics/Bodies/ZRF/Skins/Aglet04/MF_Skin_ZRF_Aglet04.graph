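# Material function for the Aglet04 skin on the ZRF vehicle body, adding animated brightening to its base color.
# Two panning samples of a shared blurred texture, one with a sine offset, are summed and masked by the skin texture's alpha.
# That mask blends toward a heavily boosted base color, whose desaturated, clamped inverse drives roughness.
# Both results override BaseColor and Roughness on MF_Vehicle_Body_Base attributes.
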MaterialFunction /VehicleCosmetics/Bodies/ZRF/Skins/Aglet04/MF_Skin_ZRF_Aglet04
graph {nodes [
  n1 = TextureSample() {desc=BC, texture=&/T_Aglet04_BC}
  n2 = Time()
  n3 = ConstantBiasScale(None=Sine(None=@n2) {period=10})
  n4 = Panner(Time=@n2) {constCoordinate=1, speedY=0.5}
  n5 = ConstantBiasScale(None=TextureSample(UVs=Add(A=@n3, B=@n4)) {texture=/VehicleCosmetics/SharedTextures/T_FinBlurred_RGB}.R) {bias=-0.1, scale=2}
  n6 = ConstantBiasScale(None=TextureSample(UVs=Panner(Time=@n2) {constCoordinate=1, speedY=0.2}) {texture=&T_FinBlurred_RGB}.G) {bias=-0.1, scale=1.5}
  n7 = Multiply(A=@n1.A, B=Add(A=@n5, B=@n6))
  n8 = LinearInterpolate(A=@n1.RGB, B=Multiply(A=@n1.RGB) {constB=20, desc=Brightening effect}, Alpha=Saturate(None=Multiply(A=@n1.A, B=@n7)))
  n9 = Desaturation(None=@n8)
  n10 = OneMinus(None=Clamp(None=Multiply(A=@n9) {constB=8}) {maxDefault=0.8, minDefault=0.7})
  FunctionOutput(None=SetMaterialAttributes!(
    MaterialAttributes=MaterialFunctionCall() {materialFunction=/VehicleCosmetics/Materials/MaterialFunctions/MF_Vehicle_Body_Base}.Result
    BaseColor=@n8
    Roughness=Saturate(None=@n10)
   ) {attributeSetTypes [#1, #2]})
  Multiply(A=@n5, B=@n6)
 ]}
